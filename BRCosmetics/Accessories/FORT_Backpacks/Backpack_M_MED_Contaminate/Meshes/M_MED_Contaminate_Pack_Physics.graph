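# Physics asset for the Contaminate medium male backpack cosmetic, defining collision and joints for its sheath and dangling strap bones.
# Single capsules cover the sheath master, strap master and three strap segments.
# Constraints chain the straps from the sheath, with fully free rotation at the strap master.
# Segment joints limit twist tightly and bend mostly along one swing axis, with that range widening toward the tip.
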PhysicsAsset /BRCosmetics/Accessories/FORT_Backpacks/Backpack_M_MED_Contaminate/Meshes/M_MED_Contaminate_Pack_Physics
bodies {
 dyn_sheath_mstr [{shapeType=Capsule, length=32.4688, radius=3.5435, center=<-5.3136 6.3274 1.0054>, rotation=<rot -0.0234 91.683 -89.7314>}]
 dyn_strap_01 [{shapeType=Capsule, length=5, radius=5, shapeName=dyn_strap_01_capsule, center=<1 0 0>, rotation=<rot 0 0 90>}]
 dyn_strap_02 [{shapeType=Capsule, length=2, radius=5, shapeName=dyn_strap_02_capsule, center=<1 0 0>, rotation=<rot 0 0 90>}]
 dyn_strap_03 [{shapeType=Capsule, length=3.5, radius=2.5, shapeName=dyn_strap_03_capsule, center=<1 0 0>, rotation=<rot 0 0 90>}]
 dyn_strap_mstr [{shapeType=Capsule, length=0.505, radius=3.5, shapeName=dyn_strap_mstr_capsule, rotation=<rot 0 0 90>}]
}
constraints [
 {bone1Name=dyn_strap_03, bone2Name=dyn_strap_02, swing1LimitDegrees=5, swing2LimitDegrees=60, twistLimitDegrees=5}
 {bone1Name=dyn_strap_01, bone2Name=dyn_strap_mstr, swing1LimitDegrees=10, swing2LimitDegrees=15, twistLimitDegrees=2}
 {bone1Name=dyn_strap_02, bone2Name=dyn_strap_01, swing1LimitDegrees=2, swing2LimitDegrees=30, twistLimitDegrees=2}
 {
  bone1Name=dyn_strap_mstr
  bone2Name=dyn_sheath_mstr
  swing1LimitDegrees=45
  swing1Motion=Free
  swing2LimitDegrees=45
  swing2Motion=Free
  twistLimitDegrees=45
  twistMotion=Free
 }
]
