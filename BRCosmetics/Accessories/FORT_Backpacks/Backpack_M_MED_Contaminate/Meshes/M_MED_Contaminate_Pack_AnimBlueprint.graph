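# Anim Blueprint for the Contaminate backpack, adjusting pack pose and adding secondary physics motion.
# Bone modifiers offset the pack for skydiving, surface swimming and ziplining, with spine-twist compensation while skydiving.
# RigidBody simulates the pack's physics asset, then a constraint and AnimDynamics drive the zipper.
# Weapon and skydive events shrink the handle bone when a Contaminate-tagged pickaxe is equipped outside skydiving.
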AnimBlueprint /BRCosmetics/Accessories/FORT_Backpacks/Backpack_M_MED_Contaminate/Meshes/M_MED_Contaminate_Pack_AnimBlueprint
target_skeleton=&/M_MED_Contaminate_Pack_Skeleton
graphs [
 AnimGraph {nodes [
   n1 = "Transform (Modify) Bone - Bone: pack"!(ComponentPose=Local To Component(LocalPose=Input Pose()), bAlphaBoolEnabled:bool, Translation="0,-2,0", Rotation="-0.000,-5.625,8.438") -> (Pose) {
    alphaInputType=Bool
    rotationMode=BMM_Additive
    translationMode=BMM_Additive
    alphaBoolBlend {blendOption=Linear}
    boneToModify {boneName=pack}
    rotation {roll=8.438, yaw=-5.625}
    translation {y=-2}
   }
   n2 = "Transform (Modify) Bone - Bone: dyn_handle"(ComponentPose=@n1, Alpha=Get HideHandle() -> (HideHandle:double), Translation="0,0,0", Rotation="0,0,0", Scale="0.01,0.01,0.01") {
    scaleMode=BMM_Replace
    scaleSpace=BCS_BoneSpace
    alphaBoolBlend {blendOption=Linear}
    boneToModify {boneName=dyn_handle}
    scale=<0.01 0.01 0.01>
   }
   n3 = "Transform (Modify) Bone - Bone: dyn_sheath_mstr"(ComponentPose=@n2, Alpha=1, Translation="0,-5,2", Rotation="0,0,0", Scale="1,1,1") {
    translationMode=BMM_Additive
    translationSpace=BCS_BoneSpace
    alphaBoolBlend {blendOption=Linear}
    boneToModify {boneName=dyn_sheath_mstr}
    translation {y=-5, z=2}
   }
   n4 = "Transform (Modify) Bone - Bone: pack"!(
    ComponentPose=@n3
    Alpha:float=Get PackAlpha() -> (PackAlpha:double)
    Translation=Get backpack_translation()
    Rotation=Get backpack_rotate()
    Scale:Vector
   ) -> (Pose) {
    rotationMode=BMM_Additive
    scaleMode=BMM_Additive
    scaleSpace=BCS_BoneSpace
    translationMode=BMM_Additive
    alphaBoolBlend {blendOption=Linear}
    boneToModify {boneName=pack}
   }
   n5 = RigidBody!(ComponentPose=@n4, Alpha, OverrideWorldGravity:Vector="0,0,-980", ExternalForce) -> (Pose) {
    bOverrideWorldGravity=true
    overridePhysicsAsset=&/M_MED_Contaminate_Pack_Physics
    simulationSpace=BaseBoneSpace
    alphaBoolBlend {blendOption=Linear}
    baseBoneRef {boneName=root}
    componentAppliedLinearAccClamp=<1500 1500 1500>
    componentLinearAccScale=<0.4 0.4 0.4>
    componentLinearVelScale=<0.4 0.4 0.4>
    overrideWorldGravity {z=-980}
   }
   n6 = "Constraint - dyn_zipper (T)"(ComponentPose=@n5, Alpha=1, ConstraintWeights_0=1) {
    alphaBoolBlend {blendOption=Linear}
    boneToModify {boneName=dyn_zipper}
    constraintSetup [{transformType=Translation, targetBone {boneName=dyn_zipper_mstr}}]
    constraintWeights [1]
   }
   n7 = Anim Dynamics - Bone: dyn_zipper(ComponentPose=@n6, Alpha=1) {
    angularSpringConstant=500
    bAngularSpring=true
    bUseGravityOverride=true
    simulationSpace=BoneRelative
    alphaBoolBlend {blendOption=Linear}
    boundBone {boneName=dyn_zipper}
    componentAppliedLinearAccClamp=<5000 5000 5000>
    componentLinearAccScale=<0.5 0.5 0.5>
    componentLinearVelScale=<0.5 0.5 0.5>
    gravityOverride {z=-980}
    physicsBodyDefinitions [{
      boundBone {boneName=dyn_zipper}
      boxExtents=<10 10 10>
      constraintSetup {angularLimitsMax=<5 130 5>, angularLimitsMin=<-5 -5 -5>, angularTarget=<0 0 0>}
      localJointOffset=<1 0 0>
     }]
    relativeSpaceBone {boneName=pack}
   }
   n8 = "Transform (Modify) Bone - Bone: pack"(ComponentPose=@n7, Alpha=Get Is Skydiving Alpha() -> (Is Skydiving Alpha:double), Translation="0,0,0", Rotation="-10,-10,0") {
    rotationMode=BMM_Additive
    rotationSpace=BCS_BoneSpace
    translationMode=BMM_Additive
    translationSpace=BCS_BoneSpace
    alphaBoolBlend {blendOption=Linear}
    boneToModify {boneName=pack}
    rotation {pitch=-10, yaw=-10}
   }
   n9 = "Transform (Modify) Bone - Bone: pack"(ComponentPose=@n8, Alpha=Get Is Swimming() -> (Is Swimming:double), Translation="0,0,0", Rotation="0,-10,-10") {
    rotationMode=BMM_Additive
    rotationSpace=BCS_BoneSpace
    translationMode=BMM_Additive
    translationSpace=BCS_BoneSpace
    alphaBoolBlend {blendOption=Linear}
    boneToModify {boneName=pack}
    rotation {roll=-10, yaw=-10}
   }
   n10 = "Transform (Modify) Bone - Bone: pack"(ComponentPose=@n9, Alpha=Get Is Ziplining alpha() -> (Is Ziplining alpha:double), Translation="0,0,0", Rotation="0,-10,-10") {
    rotationMode=BMM_Additive
    rotationSpace=BCS_BoneSpace
    translationMode=BMM_Additive
    translationSpace=BCS_BoneSpace
    alphaBoolBlend {blendOption=Linear}
    boneToModify {boneName=pack}
    rotation {roll=-10, yaw=-10}
   }
   Output Pose(Result=Component To Local(ComponentPose=@n10))
  ]}
 EventGraph {nodes [
   Event BlueprintUpdateAnimation()
   > n1 = Sequence!(execute) -> (then_0, then_1, then_2:Exec)
   n2 = Gravity Override State Function DEPRECATED(
    execute=@n1.then_1
    CharacterPartAnimInstance=Self-Reference()
    GameTimeInSeconds=GetGameTimeInSeconds()
    GravityOverrideParams=Get Gravity Override Params()
   )
   > Set GravityOverride(GravityOverride=@n2)
   > Set Is Swimming(Is Swimming:double="To Float (Boolean)"(InBool=Get bIsSurfaceSwimming())) -> (Output_Get:double)
   n3 = GetOwnerSkeletalMeshComponent()
   n4 = BreakRotator(InRot=BreakTransform(InTransform=GetDeltaTransformFromRefPose(self=@n3, BoneName=spine_05, BaseName=spine_01)).Rotation)
   Set backpack_rotate(execute=@n1.then_0, backpack_rotate=MakeRotator(Roll=MapRangeClamped(Value=@n4.Yaw, InRangeA=0, InRangeB=-30, OutRangeA=3, OutRangeB=25)))
   > Set backpack_translation(backpack_translation=MakeVector(Y=MapRangeClamped(Value=@n4.Yaw, InRangeA=0, InRangeB=-30, OutRangeB=-8)))
   > n5 = Branch(Condition=Get bIsSkyDiving())
   > Set PackAlpha(PackAlpha:double=1) -> (Output_Get:double)
   Set PackAlpha(execute=@n5.else, PackAlpha=0)
   Set Is Skydiving Alpha(execute=@n1.then_2, Is Skydiving Alpha:double="To Float (Boolean)"(InBool=Get bIsSkyDiving())) -> (Output_Get:double)
   n6 = Event OnWeaponChanged()
   n7 = Make Array("[0]"=(TagName="Cosmetics.MetaTags.Contaminate"))
   Set isNotCharacterContaminate(execute=[@n6.then, @n15.then], isNotCharacterContaminate:bool=NOT Boolean(A=CharacterItemDefinitionHasAnyMetaTag(Tags=@n7))) -> (Output_Get:bool)
   n8 = PickaxeItemDefinitionHasAnyMetaTag(Tags=@n7, bRequiresEquipped=true)
   > n9 = Branch(Condition=@n8)
   > n10 = Branch(Condition=Get bIsSkyDiving())
   Set HideHandle(execute=[@n13.then, @n10.then, @n14.else, @n9.else], HideHandle:double=0) -> (Output_Get:double)
   Set HideHandle(execute=[@n10.else, @n14.then], HideHandle=1)
   n11 = Event BlueprintProcessAnimInputEvent()
   > n12 = Branch(Condition=IsNamed(self=@n11.AnimInputEvent, InEventName=aie.Player.Zipline.Entered))
   > Set Is Ziplining alpha(Is Ziplining alpha:double=1) -> (Output_Get:double)
   Set Is Ziplining alpha(execute=@n12.else, Is Ziplining alpha=0)
   n13 = Event OnBeginSkydiving()
   Event OnEndSkydiving()
   > n14 = Branch(Condition=@n8)
   n15 = Event OnCharacterCustomizationLoaded()
   TryGetPawnOwner()
   IsNamed(self=@n11.AnimInputEvent, InEventName=aie.Player.Zipline.Exited)
   Branch()
  ]}
]
variables [
 backpack_rotate {type=Rotator}
 backpack_translation {type=Vector}
 Gravity Override Params {type=GravityOverrideParamsStruct}
 GravityOverride {type=Vector}
 HideHandle {type=number}
 Is Skydiving Alpha {type=number}
 Is Swimming {type=number}
 Is Ziplining alpha {type=number}
 isNotCharacterContaminate {type=boolean}
 PackAlpha {type=number}
]
